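# Physics asset for the Stone Viper Female two-handed pickaxe, providing bodies and joints to simulate three dangling or swaying snakes on the axe.
# One large capsule covers the handle, while each snake is a chain of small capsules from a master bone to a head, with differing segment counts.
# Segment joints have tight swing and twist limits, while each master bone joins the handle with free rotation.
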PhysicsAsset /BRCosmetics/Weapons/FORT_Melee/Pickaxe_Stone_Viper_Female/Meshes/Stone_Viper_Female_Axe_PhysicsAsset
bodies {
 Melee_TwoHanded_Handle [{shapeType=Capsule, length=68.2042, radius=14.172, center=<-0.0164 -1.4741 27.3498>, rotation=<rot -8.4976 -90 0.5563>}]
 dyn_snake_a_01 [{shapeType=Capsule, length=2, radius=1.5, center=<1.5 0 0>, rotation=<rot 90 0 0>}]
 dyn_snake_a_02 [{shapeType=Capsule, length=2, radius=1.5, center=<1.5 0 0>, rotation=<rot 90 0 0>}]
 dyn_snake_a_03 [{shapeType=Capsule, length=2, radius=1.5, center=<1.5 0 0>, rotation=<rot 90 0 0>}]
 dyn_snake_a_04 [{shapeType=Capsule, length=2, radius=1.5, center=<1.5 0 0>, rotation=<rot 90 0 0>}]
 dyn_snake_a_05 [{shapeType=Capsule, length=2, radius=1.5, center=<1.5 0 0>, rotation=<rot 90 0 0>}]
 dyn_snake_a_MSTR [{shapeType=Capsule, length=0.505, radius=0.505}]
 dyn_snake_a_head [{shapeType=Capsule, length=2, radius=1.5, center=<1.5 0 0>, rotation=<rot 90 0 0>}]
 dyn_snake_b_01 [{shapeType=Capsule, length=2, radius=1.5, center=<1.5 0 0>, rotation=<rot 90 0 0>}]
 dyn_snake_b_02 [{shapeType=Capsule, length=2, radius=1.5, center=<1.5 0 0>, rotation=<rot 90 0 0>}]
 dyn_snake_b_03 [{shapeType=Capsule, length=2, radius=1.5, center=<1.5 0 0>, rotation=<rot 90 0 0>}]
 dyn_snake_b_04 [{shapeType=Capsule, length=2, radius=1.5, center=<1.5 0 0>, rotation=<rot 90 0 0>}]
 dyn_snake_b_05 [{shapeType=Capsule, length=2, radius=1.5, center=<1.5 0 0>, rotation=<rot 90 0 0>}]
 dyn_snake_b_06 [{shapeType=Capsule, length=2, radius=1.5, center=<1.5 0 0>, rotation=<rot 90 0 0>}]
 dyn_snake_b_MSTR [{shapeType=Capsule, length=0.505, radius=0.505}]
 dyn_snake_b_head [{shapeType=Capsule, length=2, radius=1.5, center=<1.5 0 0>, rotation=<rot 90 0 0>}]
 dyn_snake_c_01 [{shapeType=Capsule, length=2, radius=1.5, center=<1.5 0 0>, rotation=<rot 90 0 0>}]
 dyn_snake_c_02 [{shapeType=Capsule, length=2, radius=1.5, center=<1.5 0 0>, rotation=<rot 90 0 0>}]
 dyn_snake_c_03 [{shapeType=Capsule, length=2, radius=1.5, center=<1.5 0 0>, rotation=<rot 90 0 0>}]
 dyn_snake_c_MSTR [{shapeType=Capsule, length=0.505, radius=0.505}]
 dyn_snake_c_head [{shapeType=Capsule, length=2, radius=1.5, center=<1.5 0 0>, rotation=<rot 90 0 0>}]
}
constraints [
 {bone1Name=dyn_snake_a_head, bone2Name=dyn_snake_a_05, swing1LimitDegrees=5, swing2LimitDegrees=5, twistLimitDegrees=3}
 {bone1Name=dyn_snake_a_01, bone2Name=dyn_snake_a_MSTR, swing1LimitDegrees=5, swing2LimitDegrees=5, twistLimitDegrees=3}
 {bone1Name=dyn_snake_a_02, bone2Name=dyn_snake_a_01, swing1LimitDegrees=5, swing2LimitDegrees=5, twistLimitDegrees=3}
 {bone1Name=dyn_snake_a_03, bone2Name=dyn_snake_a_02, swing1LimitDegrees=5, swing2LimitDegrees=5, twistLimitDegrees=3}
 {bone1Name=dyn_snake_a_04, bone2Name=dyn_snake_a_03, swing1LimitDegrees=5, swing2LimitDegrees=5, twistLimitDegrees=3}
 {bone1Name=dyn_snake_a_05, bone2Name=dyn_snake_a_04, swing1LimitDegrees=5, swing2LimitDegrees=5, twistLimitDegrees=3}
 {bone1Name=dyn_snake_b_head, bone2Name=dyn_snake_b_06, swing1LimitDegrees=5, swing2LimitDegrees=5, twistLimitDegrees=3}
 {bone1Name=dyn_snake_b_06, bone2Name=dyn_snake_b_05, swing1LimitDegrees=5, swing2LimitDegrees=5, twistLimitDegrees=3}
 {bone1Name=dyn_snake_b_05, bone2Name=dyn_snake_b_04, swing1LimitDegrees=5, swing2LimitDegrees=5, twistLimitDegrees=3}
 {bone1Name=dyn_snake_b_04, bone2Name=dyn_snake_b_03, swing1LimitDegrees=5, swing2LimitDegrees=5, twistLimitDegrees=3}
 {bone1Name=dyn_snake_b_03, bone2Name=dyn_snake_b_02, swing1LimitDegrees=5, swing2LimitDegrees=5, twistLimitDegrees=3}
 {bone1Name=dyn_snake_b_02, bone2Name=dyn_snake_b_01, swing1LimitDegrees=5, swing2LimitDegrees=5, twistLimitDegrees=3}
 {bone1Name=dyn_snake_b_01, bone2Name=dyn_snake_b_MSTR, swing1LimitDegrees=5, swing2LimitDegrees=5, twistLimitDegrees=3}
 {bone1Name=dyn_snake_c_01, bone2Name=dyn_snake_c_MSTR, swing1LimitDegrees=5, swing2LimitDegrees=5, twistLimitDegrees=3}
 {bone1Name=dyn_snake_c_02, bone2Name=dyn_snake_c_01, swing1LimitDegrees=5, swing2LimitDegrees=5, twistLimitDegrees=3}
 {bone1Name=dyn_snake_c_03, bone2Name=dyn_snake_c_02, swing1LimitDegrees=5, swing2LimitDegrees=5, twistLimitDegrees=3}
 {bone1Name=dyn_snake_c_head, bone2Name=dyn_snake_c_03, swing1LimitDegrees=5, swing2LimitDegrees=5, twistLimitDegrees=3}
 {
  bone1Name=dyn_snake_a_MSTR
  bone2Name=Melee_TwoHanded_Handle
  swing1LimitDegrees=45
  swing1Motion=Free
  swing2LimitDegrees=45
  swing2Motion=Free
  twistLimitDegrees=45
  twistMotion=Free
 }
 {
  bone1Name=dyn_snake_b_MSTR
  bone2Name=Melee_TwoHanded_Handle
  swing1LimitDegrees=45
  swing1Motion=Free
  swing2LimitDegrees=45
  swing2Motion=Free
  twistLimitDegrees=45
  twistMotion=Free
 }
 {
  bone1Name=dyn_snake_c_MSTR
  bone2Name=Melee_TwoHanded_Handle
  swing1LimitDegrees=45
  swing1Motion=Free
  swing2LimitDegrees=45
  swing2Motion=Free
  twistLimitDegrees=45
  twistMotion=Free
 }
]
