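# Animation Blueprint for the Stone Viper Female pickaxe, adding secondary physics motion to the weapon's animation.
# The looping sequence feeds an additive yaw offset on the two-handed handle bone, then a base-bone-space RigidBody node with custom gravity and damped component motion.
# Rigid Alpha blends the simulation, though both front-end branches in the update event set it to the same half weight.
# A second, bool-driven RigidBody node is left unconnected.
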AnimBlueprint /BRCosmetics/Weapons/FORT_Melee/Pickaxe_Stone_Viper_Female/Meshes/Stone_Viper_Female_Axe_AnimBP
target_skeleton=&/Stone_Viper_Female_Axe_Skeleton
graphs [
 AnimGraph {nodes [
   Input Pose()
   RigidBody!(ComponentPose, bAlphaBoolEnabled:bool, ExternalForce) -> (Pose) {
    alphaInputType=Bool
    bClampLinearTranslationLimitToRefPose=true
    bOverrideWorldGravity=true
    overridePhysicsAsset=&/Stone_Viper_Female_Axe_PhysicsAsset
    simulationSpace=BaseBoneSpace
    alphaBoolBlend {blendInTime=0.5, blendOption=Linear, blendOutTime=0.5}
    baseBoneRef {boneName=b_AR_Root}
    componentAppliedLinearAccClamp=<2000 2000 1000>
    componentLinearAccScale=<0.4 0.4 0.2>
    componentLinearVelScale=<0.4 0.4 0.2>
    overrideWorldGravity {z=125}
   }
   n1 = "Transform (Modify) Bone - Bone: Melee_TwoHanded_Handle"(
    ComponentPose=Local To Component(LocalPose=Sequence Player 'Anim_Pickaxe_Stone_Viper_Female'() {sequence=/BRCosmetics/Weapons/FORT_Melee/Pickaxe_Stone_Viper_Female/Animation/Anim_Pickaxe_Stone_Viper_Female})
    Alpha=1
    Translation="0,0,0"
    Rotation="0,-30,0"
    Scale="1,1,1"
   ) {
    rotationMode=BMM_Additive
    rotationSpace=BCS_BoneSpace
    alphaBoolBlend {blendOption=Linear}
    boneToModify {boneName=Melee_TwoHanded_Handle}
    rotation {yaw=-30}
   }
   n2 = RigidBody!(ComponentPose=@n1, Alpha:float=Get Rigid Alpha() -> (Rigid Alpha:double), ExternalForce) -> (Pose) {
    bClampLinearTranslationLimitToRefPose=true
    bOverrideWorldGravity=true
    overridePhysicsAsset=&Stone_Viper_Female_Axe_PhysicsAsset
    simulationSpace=BaseBoneSpace
    alphaBoolBlend {blendOption=Linear}
    baseBoneRef {boneName=b_AR_Root}
    componentAppliedLinearAccClamp=<300 300 300>
    componentLinearAccScale=<0.3 0.3 0.3>
    componentLinearVelScale=<0.15 0.15 0.15>
    overrideWorldGravity {z=-300}
   }
   Output Pose(Result=Component To Local(ComponentPose=@n2)) {layerGroup=None}
  ]}
 EventGraph {nodes [
   Event BlueprintUpdateAnimation()
   > n1 = Branch(Condition=Get bIsInFrontEnd())
   > Set Rigid Alpha(Rigid Alpha:double=0.5) -> (Output_Get:double)
   Set Rigid Alpha(execute=@n1.else, Rigid Alpha=0.5)
   TryGetPawnOwner()
  ]}
]
variables [Rigid Alpha {type=number}]
